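# Material function that warps input UVs into a bulge, with parameters for center, scale, reversed direction and horizontal-only or vertical-only axes.
# The centered, scaled UV vector is weighted by a mask and subtracted from the UVs, with optional panning afterward.
# A static switch takes that mask from a tiled SDF texture or a procedural dot-product falloff.
# It outputs the warped UVs plus soft and smoothstepped solid masks.
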MaterialFunction /Game/Global/Materials/MaterialGym/Materials/MaterialElements/UVs/UVMainpulation/BulgeUV/MF_BulgeUV
graph {nodes [
  n1 = TextureSampleParameter2D(UVs=Multiply(A=TextureCoordinate(), B=ScalarParameter() {defaultValue=1, group=Bulge_UV, parameterName=BulgeSDFTiling})) {
   group=Bulge_UV
   parameterName=BulgeSDFTexture
   samplerSource=SSM_Wrap_WorldGroupSettings
   samplerType=SAMPLERTYPE_LinearColor
   texture=/Game/Global/Materials/MaterialGym/Materials/MaterialElements/UVs/UVMainpulation/SphereUV/T_heart_SingleChannelDF
  }
  n2 = FunctionInput(Preview=TextureCoordinate()) {bUsePreviewValueAsDefault=true, inputName=UV, inputType=FunctionInput_Vector2}
  n3 = ScalarParameter() {defaultValue=-0.5, group=Bulge_UV, parameterName=BulgeCenterBias}
  n4 = ScalarParameter() {defaultValue=3, group=Bulge_UV, parameterName=BulgeUVScale}
  n5 = Multiply(A=@n4, B=Constant() {r=-1})
  n6 = ScalarParameter() {group=Bulge_UV, parameterName=Bulge Reverse direction}
  MaterialExpressionConstant_5 = Constant() {r=1}
  MaterialExpressionAppendVector_2 = AppendVector(A=Constant() {r=1}, B=Constant())
  MaterialExpressionAppendVector_3 = AppendVector(A=Constant(), B=Constant() {r=1})
  n7 = Switch!(
   SwitchValue=ScalarParameter() {group=Bulge_UV, parameterName=BulgeModeSelect}
   Default=@MaterialExpressionConstant_5
   Normal=@MaterialExpressionConstant_5
   HorizontalOnly=@MaterialExpressionAppendVector_2
   VerticalOnly=@MaterialExpressionAppendVector_3
  ) {inputs [
    {
     inputName=Normal
     input {
      expression=&self:MaterialExpressionConstant_5
      inputName=None
      mask=0
      maskA=0
      maskB=0
      maskG=0
      maskR=0
      outputIndex=0
     }
    }
    {
     inputName=HorizontalOnly
     input {
      expression=&self:MaterialExpressionAppendVector_2
      inputName=None
      mask=0
      maskA=0
      maskB=0
      maskG=0
      maskR=0
      outputIndex=0
     }
    }
    {
     inputName=VerticalOnly
     input {
      expression=&self:MaterialExpressionAppendVector_3
      inputName=None
      mask=0
      maskA=0
      maskB=0
      maskG=0
      maskR=0
      outputIndex=0
     }
    }
   ]}
  n8 = Multiply(A=Multiply(A=Add(A=@n2, B=@n3), B=LinearInterpolate(A=@n4, B=@n5, Alpha=@n6)), B=@n7)
  n9 = OneMinus(None=Saturate(None=DotProduct(A=@n8, B=@n8)))
  n10 = StaticSwitchParameter(True=@n1.R, False=@n9) {defaultValue=true, group=Bulge_UV, parameterName=UseSDFTexture}
  n11 = Subtract(A=@n2, B=Multiply(A=Multiply(A=@n8, B=@n10), B=ScalarParameter() {defaultValue=0.1, group=Bulge_UV, parameterName=BulgeIntensityOffset}))
  n12 = AppendVector(
   A=ScalarParameter() {group=Bulge_UV, parameterName=Bulge panner AnimationSpeedX}
   B=ScalarParameter() {defaultValue=0.2, group=Bulge_UV, parameterName=Bulge Panner AnimationSpeedY}
  )
  FunctionOutput(None=StaticSwitchParameter(True=Panner(Coordinate=@n11, Speed=@n12) {speedY=0.3}, False=@n11) {group=Bulge_UV, parameterName=BulgePannerAnimation?}) {outputName=BulgeUV}
  FunctionOutput(None=NamedRerouteUsage()) {outputName=OutputMask, sortPriority=3}
  FunctionOutput(None=NamedRerouteUsage()) {outputName=outputSolidMask, sortPriority=2}
  NamedRerouteDeclaration(None=StaticSwitchParameter(True=@n1.R, False=@n9) {defaultValue=true, group=Bulge_UV, parameterName=UseSDFTexture_Bulge}) {name=Output mask, nodeColor=<rgba 1 0.5625 0 1>}
  n13 = ScalarParameter() {defaultValue=0.001, group=Bulge_UV, parameterName=Bulge solid mask softness}
  n14 = ScalarParameter() {defaultValue=0.001, group=Bulge_UV, parameterName=Bulge solid mask softness}
  NamedRerouteDeclaration(None=StaticSwitchParameter(True=SmoothStep(Max=@n13, Value=@n1.R), False=SmoothStep(Max=@n14, Value=@n9)) {defaultValue=true, group=Bulge_UV, parameterName=UseSDFTexture}) {name=Output solid mask, nodeColor=<rgba 0.0781 1 0 1>}
 ]}
