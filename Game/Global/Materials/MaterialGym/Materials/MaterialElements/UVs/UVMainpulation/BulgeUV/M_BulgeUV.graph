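# M_BulgeUV is a Material Gym material that demonstrates the MF_BulgeUV UV-manipulation function.
# The function's BulgeUV output drives the UVs of a texture parameter, defaulting to an editor color grid.
# A DebugMask_Bulge static switch chooses between the plain sampled color and a lerp from black by the function's outputSolidMask, showing the bulge mask.
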Material /Game/Global/Materials/MaterialGym/Materials/MaterialElements/UVs/UVMainpulation/BulgeUV/M_BulgeUV
graph {
 properties {bAutomaticallySetUsageInEditor=false, usedShadingModels=Default Lit}
 nodes [
  n1 = MaterialFunctionCall!(UV) {materialFunction=&/MF_BulgeUV}
  n2 = TextureSampleParameter2D(UVs=@n1.BulgeUV) {group=Bulge_UV, parameterName=BulgeUVTexture, samplerSource=SSM_Wrap_WorldGroupSettings, texture=/UVEditor/Textures/UVEditorColorGrid_Color}
  StaticSwitchParameter(True=LinearInterpolate(A=Constant3Vector() {constant=<rgba 0 0 0 1>}, B=@n2.RGB, Alpha=@n1.outputSolidMask), False=@n2.RGB) {group=Bulge_UV, parameterName=DebugMask_Bulge}
 ]
}
